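# Physics asset for the Scruffy Rider male medium body cosmetic, setting up secondary motion for its collar, scarf and right-arm armband knot.
# Each dyn_ capsule hangs from a small base_ anchor through swing and twist angle limits tuned per piece.
# Anchors link to spine_04 or upperarm_r with free angular motion.
# Capsules on head, upper spine, upper arms and right clavicle supply body shapes around the dangling parts.
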PhysicsAsset /BRCosmetics/Characters/Player/Male/Medium/Bodies/M_MED_Rider_Scruffy/Meshes/M_MED_Rider_Scruffy_Physics
bodies {
 base_armband_knot [{shapeType=Capsule, length=1, radius=1, shapeName=base_armband_knot_capsule, rotation=<rot -90 0 0>}]
 base_collar_back [{shapeType=Capsule, length=1, radius=1, shapeName=base_collar_back_capsule, rotation=<rot -90 0 0>}]
 base_collar_l [{shapeType=Capsule, length=1, radius=1, shapeName=base_collar_l_capsule, rotation=<rot -90 0 0>}]
 base_collar_r [{shapeType=Capsule, length=1, radius=1, shapeName=base_collar_r_capsule, rotation=<rot -90 0 0>}]
 base_scarf_back [{shapeType=Capsule, length=1, radius=1, shapeName=base_scarf_back_capsule, rotation=<rot -90 0 0>}]
 base_scarf_f [{shapeType=Capsule, length=1, radius=1, shapeName=base_scarf_f_capsule, rotation=<rot -90 0 0>}]
 base_scarf_l [{shapeType=Capsule, length=1, radius=1, shapeName=base_scarf_l_capsule, rotation=<rot -90 0 0>}]
 base_scarf_low [{shapeType=Capsule, length=1, radius=1, shapeName=base_scarf_low_capsule, rotation=<rot -90 0 0>}]
 base_scarf_r [{shapeType=Capsule, length=1, radius=1, shapeName=base_scarf_r_capsule, rotation=<rot -90 0 0>}]
 clavicle_r [{shapeType=Capsule, length=10, radius=1.25, shapeName=clavicle_r_capsule, center=<-15 0 -2>, rotation=<rot 0 -90 -90>}]
 dyn_armband_knot [{shapeType=Capsule, length=1.5, radius=4, shapeName=dyn_armband_knot_capsule, center=<3 0 0>, rotation=<rot -90 0 0>}]
 dyn_collar_back [{shapeType=Capsule, length=16, radius=2, shapeName=dyn_collar_back_capsule, center=<4 0 0>}]
 dyn_collar_l [{shapeType=Capsule, length=18, radius=2.5, shapeName=dyn_collar_l_capsule, center=<4 2 0>, rotation=<rot 0 0 -90>}]
 dyn_collar_r [{shapeType=Capsule, length=18, radius=1.5, shapeName=dyn_collar_r_capsule, center=<4 -2 0>, rotation=<rot 0 0 -90>}]
 dyn_scarf_back [{shapeType=Capsule, length=1.5, radius=1.5, shapeName=dyn_scarf_back_capsule, center=<1 0 0>, rotation=<rot -90 0 0>}]
 dyn_scarf_f [{shapeType=Capsule, length=1.5, radius=1.5, shapeName=dyn_scarf_f_capsule, center=<1 0 0>, rotation=<rot -90 0 0>}]
 dyn_scarf_l [{shapeType=Capsule, length=1.5, radius=1.5, shapeName=dyn_scarf_l_capsule, center=<1 0 0>, rotation=<rot -90 0 0>}]
 dyn_scarf_low [{shapeType=Capsule, length=1.5, radius=1.5, shapeName=dyn_scarf_low_capsule, center=<1 0 0>, rotation=<rot -90 0 0>}]
 dyn_scarf_r [{shapeType=Capsule, length=1.5, radius=1.5, shapeName=dyn_scarf_r_capsule, center=<1 0 0>, rotation=<rot -90 0 0>}]
 head [{shapeType=Capsule, length=4, radius=7, shapeName=head_capsule, rotation=<rot -90 0 0>}]
 spine_04 [{shapeType=Capsule, length=9, radius=12, center=<0 -2 0>}]
 upperarm_l [{shapeType=Capsule, length=20, radius=4.6, shapeName=upperarm_r_capsule, center=<10.7614 0 0>, rotation=<rot -90 180 180>}]
 upperarm_r [{shapeType=Capsule, length=6, radius=5.5, shapeName=upperarm_r_capsule, center=<-5 0 0>, rotation=<rot -90 30 0>}]
}
constraints [
 {bone1Name=dyn_collar_back, bone2Name=base_collar_back, swing1LimitDegrees=20, swing2LimitDegrees=2, twistLimitDegrees=5}
 {bone1Name=dyn_collar_l, bone2Name=base_collar_l, swing1LimitDegrees=2, swing2LimitDegrees=30, twistLimitDegrees=5}
 {bone1Name=dyn_collar_r, bone2Name=base_collar_r, swing1LimitDegrees=0, swing2LimitDegrees=20, twistLimitDegrees=0}
 {bone1Name=dyn_scarf_f, bone2Name=base_scarf_f, swing1LimitDegrees=45, swing2LimitDegrees=45, twistLimitDegrees=25}
 {bone1Name=dyn_scarf_l, bone2Name=base_scarf_l, swing1LimitDegrees=2, swing2LimitDegrees=10, twistLimitDegrees=5}
 {bone1Name=dyn_scarf_low, bone2Name=base_scarf_low, swing1LimitDegrees=45, swing2LimitDegrees=45, twistLimitDegrees=25}
 {bone1Name=dyn_scarf_r, bone2Name=base_scarf_r, swing1LimitDegrees=2, swing2LimitDegrees=10, twistLimitDegrees=5}
 {bone1Name=dyn_armband_knot, bone2Name=base_armband_knot, swing1LimitDegrees=5, swing2LimitDegrees=5, twistLimitDegrees=1}
 {bone1Name=dyn_scarf_back, bone2Name=base_scarf_back, swing1LimitDegrees=10, swing2LimitDegrees=2, twistLimitDegrees=5}
 {
  bone1Name=base_armband_knot
  bone2Name=upperarm_r
  swing1LimitDegrees=45
  swing1Motion=Free
  swing2LimitDegrees=45
  swing2Motion=Free
  twistLimitDegrees=45
  twistMotion=Free
 }
 {
  bone1Name=base_scarf_l
  bone2Name=spine_04
  swing1LimitDegrees=45
  swing1Motion=Free
  swing2LimitDegrees=45
  swing2Motion=Free
  twistLimitDegrees=45
  twistMotion=Free
 }
 {
  bone1Name=base_scarf_r
  bone2Name=spine_04
  swing1LimitDegrees=45
  swing1Motion=Free
  swing2LimitDegrees=45
  swing2Motion=Free
  twistLimitDegrees=45
  twistMotion=Free
 }
 {
  bone1Name=base_collar_back
  bone2Name=spine_04
  swing1LimitDegrees=45
  swing1Motion=Free
  swing2LimitDegrees=45
  swing2Motion=Free
  twistLimitDegrees=45
  twistMotion=Free
 }
 {
  bone1Name=base_collar_r
  bone2Name=spine_04
  swing1LimitDegrees=45
  swing1Motion=Free
  swing2LimitDegrees=45
  swing2Motion=Free
  twistLimitDegrees=45
  twistMotion=Free
 }
 {
  bone1Name=base_collar_l
  bone2Name=spine_04
  swing1LimitDegrees=45
  swing1Motion=Free
  swing2LimitDegrees=45
  swing2Motion=Free
  twistLimitDegrees=45
  twistMotion=Free
 }
 {
  bone1Name=base_scarf_f
  bone2Name=spine_04
  swing1LimitDegrees=45
  swing1Motion=Free
  swing2LimitDegrees=45
  swing2Motion=Free
  twistLimitDegrees=45
  twistMotion=Free
 }
 {
  bone1Name=base_scarf_low
  bone2Name=spine_04
  swing1LimitDegrees=45
  swing1Motion=Free
  swing2LimitDegrees=45
  swing2Motion=Free
  twistLimitDegrees=45
  twistMotion=Free
 }
 {
  bone1Name=base_scarf_back
  bone2Name=spine_04
  swing1LimitDegrees=45
  swing1Motion=Free
  swing2LimitDegrees=45
  swing2Motion=Free
  twistLimitDegrees=45
  twistMotion=Free
 }
 {
  bone1Name=head
  bone2Name=spine_04
  swing1LimitDegrees=45
  swing1Motion=Free
  swing2LimitDegrees=45
  swing2Motion=Free
  twistLimitDegrees=45
  twistMotion=Free
 }
 {
  bone1Name=upperarm_r
  bone2Name=clavicle_r
  swing1LimitDegrees=45
  swing1Motion=Free
  swing2LimitDegrees=45
  swing2Motion=Free
  twistLimitDegrees=45
  twistMotion=Free
 }
 {
  bone1Name=upperarm_l
  bone2Name=spine_04
  swing1LimitDegrees=45
  swing1Motion=Free
  swing2LimitDegrees=45
  swing2Motion=Free
  twistLimitDegrees=45
  twistMotion=Free
 }
 {
  bone1Name=clavicle_r
  bone2Name=spine_04
  swing1LimitDegrees=45
  swing1Motion=Free
  swing2LimitDegrees=45
  swing2Motion=Free
  twistLimitDegrees=45
  twistMotion=Free
 }
]
