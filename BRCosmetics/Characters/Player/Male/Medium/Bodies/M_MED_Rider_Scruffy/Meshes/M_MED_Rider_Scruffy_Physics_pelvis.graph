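# Physics asset for the pelvis pouch on the Rider Scruffy medium male body cosmetic, covering only the pouch and nearby body.
# Capsule bodies sit on the pouch's base and dynamic bones, the pelvis and the left thigh, giving the pouch body shapes to collide with.
# The dynamic pouch bone has tight swing and twist limits to its base, while base and thigh link to the pelvis with free angular motion.
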PhysicsAsset /BRCosmetics/Characters/Player/Male/Medium/Bodies/M_MED_Rider_Scruffy/Meshes/M_MED_Rider_Scruffy_Physics_pelvis
bodies {
 base_pouch_pelvis [{shapeType=Capsule, length=1, radius=1, shapeName=base_pouch_pelvis_capsule, rotation=<rot -90 0 0>}]
 dyn_pouch_pelvis [{shapeType=Capsule, length=4.5, radius=2.2, shapeName=dyn_pouch_pelvis_capsule, center=<12.6 0 2.9182>, rotation=<rot -90 0 0>}]
 pelvis [{shapeType=Capsule, length=4, radius=12}]
 thigh_l [{shapeType=Capsule, length=18, radius=9, center=<-15 2 0>, rotation=<rot 95 0 0>}]
}
constraints [
 {bone1Name=dyn_pouch_pelvis, bone2Name=base_pouch_pelvis, swing1LimitDegrees=5, swing2LimitDegrees=10, twistLimitDegrees=1}
 {
  bone1Name=base_pouch_pelvis
  bone2Name=pelvis
  swing1LimitDegrees=45
  swing1Motion=Free
  swing2LimitDegrees=45
  swing2Motion=Free
  twistLimitDegrees=45
  twistMotion=Free
 }
 {
  bone1Name=thigh_l
  bone2Name=pelvis
  swing1LimitDegrees=45
  swing1Motion=Free
  swing2LimitDegrees=45
  swing2Motion=Free
  twistLimitDegrees=45
  twistMotion=Free
 }
]
