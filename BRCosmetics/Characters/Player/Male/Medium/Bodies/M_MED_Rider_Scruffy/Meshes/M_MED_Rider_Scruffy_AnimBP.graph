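# Post-process style AnimBlueprint for the M_MED_Rider_Scruffy player cosmetic body, adding secondary motion to pouches, collar, scarf and armband.
# Copies the attached parent's pose, then runs a deform Control Rig and weighted pouch constraints.
# Cached poses feed three gravity-overridden RigidBody sims and two armband AnimDynamics nodes, followed by additive corrective bone offsets.
# EventGraph derives correction alphas from head and thigh deltas, crouch, targeting, movement, weapon type and long back bling.
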AnimBlueprint /BRCosmetics/Characters/Player/Male/Medium/Bodies/M_MED_Rider_Scruffy/Meshes/M_MED_Rider_Scruffy_AnimBP
target_skeleton=&/M_MED_Rider_Scruffy_Skeleton
graphs [
 AnimGraph {nodes [
   n1 = "Transform (Modify) Bone - Bone: base_armband_knot"!(
    ComponentPose=Local To Component(LocalPose=Use cached pose 'PostSimCache'())
    bAlphaBoolEnabled:bool=Get isWearingCape2() -> (isWearingCape2:bool)
    Translation="0,0,0"
    Rotation="0,40,40"
    Scale="1,1,1"
   ) -> (Pose) {
    alphaInputType=Bool
    rotationMode=BMM_Additive
    rotationSpace=BCS_ParentBoneSpace
    translationMode=BMM_Additive
    translationSpace=BCS_ParentBoneSpace
    alphaBoolBlend {blendOption=Linear}
    alphaScaleBiasClamp {interpSpeedDecreasing=15, interpSpeedIncreasing=15}
    boneToModify {boneName=base_armband_knot}
    rotation {roll=40, yaw=40}
   }
   Output Pose(Result=Component To Local(ComponentPose=@n1)) {layerGroup=None}
   n2 = Control Rig(Source=Copy Pose From Mesh() {bUseAttachedParent=true}) {
    alphaBoolBlend {blendOption=Linear}
    controlRigAssetReference {blueprintRigClass=&/M_MED_Rider_Scruffy_DeformRig.M_MED_Rider_Scruffy_DeformRig_C}
   }
   n3 = "Constraint - base_pouch_leg (PPPPP)"!(
    ComponentPose=Local To Component(LocalPose=@n2)
    Alpha=1
    ConstraintWeights_0=0.6375
    ConstraintWeights_1:float=0.2012
    ConstraintWeights_2:float=0.0796
    ConstraintWeights_3:float=0.0699
    ConstraintWeights_4:float=0.0118
   ) -> (Pose) {
    alphaBoolBlend {blendOption=Linear}
    boneToModify {boneName=base_pouch_leg}
    constraintSetup [
     {targetBone {boneName=thigh_twist_01_r}}
     {targetBone {boneName=thigh_r}}
     {targetBone {boneName=deform_groin_r}}
     {targetBone {boneName=pelvis}}
     {targetBone {boneName=thigh_twist_01_l}}
    ]
    constraintWeights [1,1,1,1,1]
   }
   n4 = "Constraint - base_pouch_pelvis (PPP)"!(ComponentPose=@n3, Alpha=1, ConstraintWeights_0=0.8762, ConstraintWeights_1:float=0.064, ConstraintWeights_2:float=0.0598) -> (Pose) {
    alphaBoolBlend {blendOption=Linear}
    boneToModify {boneName=base_pouch_pelvis}
    constraintSetup [{targetBone {boneName=pelvis}}, {targetBone {boneName=spine_01}}, {targetBone {boneName=deform_groin_l}}]
    constraintWeights [1,1,1]
   }
   Save cached pose 'ctrlRig_pose'(Pose=Component To Local(ComponentPose=@n4)) {cacheName=ctrlRig_pose}
   n5 = "Transform (Modify) Bone - Bone: base_pouch_pelvis"(
    ComponentPose=Local To Component(LocalPose=Use cached pose 'ctrlRig_pose'())
    Alpha=Get isPouchingRotate() -> (isPouchingRotate:double)
    Translation="0,-8,-2.5"
    Rotation="-5,0,45"
    Scale="1,1,1"
   ) {
    rotationMode=BMM_Additive
    rotationSpace=BCS_ParentBoneSpace
    translationMode=BMM_Additive
    translationSpace=BCS_ParentBoneSpace
    alphaBoolBlend {blendOption=Linear}
    alphaScaleBiasClamp {bInterpResult=true, interpSpeedDecreasing=30, interpSpeedIncreasing=15}
    boneToModify {boneName=base_pouch_pelvis}
    rotation {pitch=-5, roll=45}
    translation {y=-8, z=-2.5}
   }
   n6 = Get GravityOverride()
   n7 = RigidBody!(ComponentPose=@n5, Alpha, OverrideWorldGravity:Vector=@n6) -> (Pose) {
    bClampLinearTranslationLimitToRefPose=true
    bOverrideWorldGravity=true
    overridePhysicsAsset=&/M_MED_Rider_Scruffy_Physics
    simulationSpace=BaseBoneSpace
    alphaBoolBlend {blendOption=Linear}
    baseBoneRef {boneName=spine_05}
    componentAppliedLinearAccClamp=<3000 3000 3000>
    componentLinearAccScale=<0.3 0.4 4>
    componentLinearVelScale=<0.3 4 4>
    simSpaceSettings {
     maxAngularAcceleration=1000
     maxAngularVelocity=3
     maxLinearAcceleration=700
     maxLinearVelocity=500
     worldAlpha=0.2375
     externalLinearDragV=<0.2 0.2>
    }
   }
   n8 = RigidBody(ComponentPose=@n7, Alpha=Get leftCollar_RB_dyn() -> (leftCollar_RB_dyn:double), OverrideWorldGravity=@n6) {
    bClampLinearTranslationLimitToRefPose=true
    bOverrideWorldGravity=true
    overridePhysicsAsset=&/M_MED_Rider_Scruffy_Physics_leftCollar
    simulationSpace=BaseBoneSpace
    alphaBoolBlend {blendOption=Linear}
    baseBoneRef {boneName=spine_05}
    componentAppliedLinearAccClamp=<3000 3000 3000>
    componentLinearAccScale=<0.3 0.4 4>
    componentLinearVelScale=<0.3 4 4>
    simSpaceSettings {
     maxAngularAcceleration=1000
     maxAngularVelocity=3
     maxLinearAcceleration=700
     maxLinearVelocity=500
     worldAlpha=0.2375
     externalLinearDragV=<0.2 0.2>
    }
   }
   n9 = RigidBody(ComponentPose=@n8, OverrideWorldGravity=Get GravityOverride()) {
    bClampLinearTranslationLimitToRefPose=true
    bOverrideWorldGravity=true
    overridePhysicsAsset=&/M_MED_Rider_Scruffy_Physics_pelvis
    simulationSpace=BaseBoneSpace
    alphaBoolBlend {blendOption=Linear}
    baseBoneRef {boneName=pelvis}
    componentAppliedLinearAccClamp=<3000 3000 3000>
    componentLinearAccScale=<0.3 0.2 0.2>
    componentLinearVelScale=<0.3 0.2 0.2>
    simSpaceSettings {
     maxAngularAcceleration=1000
     maxAngularVelocity=3
     maxLinearAcceleration=700
     maxLinearVelocity=500
     worldAlpha=0.2
     externalLinearDragV=<0.2 0.2>
    }
   }
   n10 = Anim Dynamics - Bone: dyn_armband_low!(ComponentPose=@n9, Alpha=1, GravityOverride:Vector="0,0,-980", ExternalForce="0,0,0") -> (Pose) {
    alpha=0.9
    angularDampingOverride=0.98
    angularSpringConstant=1500
    bAngularSpring=true
    bGravityOverrideInSimSpace=true
    bOverrideAngularDamping=true
    bUseGravityOverride=true
    numSolverIterationsPostUpdate=4
    numSolverIterationsPreUpdate=16
    simulationSpace=BoneRelative
    alphaBoolBlend {blendOption=Linear}
    boundBone {boneName=dyn_armband_low}
    componentAppliedLinearAccClamp=<60000 60000 60000>
    componentLinearAccScale=<0.3 0.2 0.2>
    componentLinearVelScale=<0.3 0.2 0.2>
    gravityOverride {z=-980}
    physicsBodyDefinitions [{
      sphereCollisionRadius=0
      boundBone {boneName=dyn_armband_low}
      boxExtents=<15 5 5>
      constraintSetup {angularTargetAxis=AxisZ, twistAxis=AxisY, angularLimitsMax=<25 25 25>, angularLimitsMin=<-25 -25 -25>, angularTarget=<0 0 1>}
      localJointOffset=<0 0 0>
     }]
    relativeSpaceBone {boneName=root}
   }
   n11 = Anim Dynamics - Bone: dyn_armband_hi!(ComponentPose=@n10, Alpha=1, GravityOverride:Vector="0,0,-980", ExternalForce="0,0,0") -> (Pose) {
    alpha=0.9
    angularDampingOverride=0.98
    angularSpringConstant=1500
    bAngularSpring=true
    bGravityOverrideInSimSpace=true
    bOverrideAngularDamping=true
    bUseGravityOverride=true
    numSolverIterationsPostUpdate=4
    numSolverIterationsPreUpdate=16
    simulationSpace=BoneRelative
    alphaBoolBlend {blendOption=Linear}
    boundBone {boneName=dyn_armband_hi}
    componentAppliedLinearAccClamp=<60000 60000 60000>
    componentLinearAccScale=<0.3 0.2 0.2>
    componentLinearVelScale=<0.3 0.2 0.2>
    gravityOverride {z=-980}
    physicsBodyDefinitions [{
      sphereCollisionRadius=0
      boundBone {boneName=dyn_armband_hi}
      boxExtents=<15 5 5>
      constraintSetup {angularTargetAxis=AxisZ, twistAxis=AxisY, angularLimitsMax=<25 25 25>, angularLimitsMin=<-25 -25 -25>, angularTarget=<0 0 1>}
      localJointOffset=<0 0 0>
     }]
    relativeSpaceBone {boneName=root}
   }
   Save cached pose 'RB_cache'(Pose=Component To Local(ComponentPose=@n11)) {cacheName=RB_cache}
   n12 = "Transform (Modify) Bone - Bone: base_collar_back"(
    ComponentPose=Local To Component(LocalPose=Use cached pose 'RB_cache'())
    Alpha=Get headUp() -> (headUp:double)
    Translation=Get translateCollarBack() -> (translateCollarBack:Vector)
    Rotation=Get rotationCollarBack() -> (rotationCollarBack:Rotator)
    Scale="1,1,1.2"
   ) {
    rotationMode=BMM_Additive
    rotationSpace=BCS_BoneSpace
    scaleMode=BMM_Additive
    scaleSpace=BCS_BoneSpace
    translationMode=BMM_Additive
    translationSpace=BCS_BoneSpace
    alphaBoolBlend {blendOption=Linear}
    alphaScaleBiasClamp {interpSpeedDecreasing=19.75, interpSpeedIncreasing=19.75}
    boneToModify {boneName=base_collar_back}
    rotation {pitch=25}
    scale {z=1.2}
    translation {x=-5}
   }
   n13 = "Transform (Modify) Bone - Bone: base_scarf_l"(ComponentPose=@n12, Alpha=Get headUp(), Translation="0,0,-3", Rotation="0,0,0", Scale="1,1,1") {
    rotationMode=BMM_Additive
    rotationSpace=BCS_BoneSpace
    scaleMode=BMM_Additive
    scaleSpace=BCS_BoneSpace
    translationMode=BMM_Additive
    translationSpace=BCS_BoneSpace
    alphaBoolBlend {blendOption=Linear}
    alphaScaleBiasClamp {bInterpResult=true}
    boneToModify {boneName=base_scarf_l}
    translation {z=-3}
   }
   n14 = "Transform (Modify) Bone - Bone: base_collar_l"(
    ComponentPose=@n13
    Alpha=Get isUsingRocketLauncherFloat() -> (isUsingRocketLauncherFloat:double)
    Translation="0,0,0"
    Rotation="0,0,-10"
    Scale="1,1,1"
   ) {
    rotationMode=BMM_Additive
    rotationSpace=BCS_BoneSpace
    scaleMode=BMM_Additive
    scaleSpace=BCS_BoneSpace
    translationMode=BMM_Additive
    translationSpace=BCS_BoneSpace
    alphaBoolBlend {blendOption=Linear}
    alphaScaleBiasClamp {interpSpeedDecreasing=15, interpSpeedIncreasing=5.5}
    boneToModify {boneName=base_collar_l}
    rotation {roll=-10}
   }
   n15 = "Transform (Modify) Bone - Bone: base_collar_r"(
    ComponentPose=@n14
    Alpha=Get isUsingDoublePistolsCrouch() -> (isUsingDoublePistolsCrouch:double)
    Translation="0,0,0"
    Rotation="-10,0,0"
    Scale="1,1,1"
   ) {
    rotationMode=BMM_Additive
    rotationSpace=BCS_BoneSpace
    scaleMode=BMM_Additive
    scaleSpace=BCS_BoneSpace
    translationMode=BMM_Additive
    translationSpace=BCS_BoneSpace
    alphaBoolBlend {blendOption=Linear}
    alphaScaleBiasClamp {bInterpResult=true, interpSpeedDecreasing=5, interpSpeedIncreasing=5}
    boneToModify {boneName=base_collar_r}
    rotation {pitch=-10}
   }
   n16 = "Transform (Modify) Bone - Bone: base_collar_r"(
    ComponentPose=@n15
    Alpha=Get isCrounchingMoving_Pistol() -> (isCrounchingMoving_Pistol:double)
    Translation="-2,0,-3"
    Rotation="5,-25,0"
    Scale="1,1,1"
   ) {
    rotationMode=BMM_Additive
    rotationSpace=BCS_BoneSpace
    scaleMode=BMM_Additive
    scaleSpace=BCS_BoneSpace
    translationMode=BMM_Additive
    translationSpace=BCS_BoneSpace
    alphaBoolBlend {blendOption=Linear}
    alphaScaleBiasClamp {bInterpResult=true, interpSpeedDecreasing=15, interpSpeedIncreasing=5}
    boneToModify {boneName=base_collar_r}
    rotation {pitch=5, yaw=-25}
    translation {x=-2, z=-3}
   }
   n17 = "Transform (Modify) Bone - Bone: base_collar_back"(
    ComponentPose=@n16
    Alpha=Get Is Targeting headUp Back() -> (Is Targeting headUp Back:double)
    Translation="0,0,0"
    Rotation="0,-15,0"
    Scale="1,1,1"
   ) {
    rotationMode=BMM_Additive
    rotationSpace=BCS_BoneSpace
    scaleMode=BMM_Additive
    scaleSpace=BCS_BoneSpace
    translationMode=BMM_Additive
    translationSpace=BCS_BoneSpace
    alphaBoolBlend {blendOption=Linear}
    alphaScaleBiasClamp {bInterpResult=true, interpSpeedDecreasing=15, interpSpeedIncreasing=15}
    boneToModify {boneName=base_collar_back}
    rotation {yaw=-15}
   }
   n18 = "Transform (Modify) Bone - Bone: base_collar_r"(ComponentPose=@n17, Alpha=Get isTargetingHeadUp() -> (isTargetingHeadUp:double), Translation="0,0,0", Rotation="0,0,25", Scale="1,1,1") {
    rotationMode=BMM_Additive
    rotationSpace=BCS_BoneSpace
    scaleMode=BMM_Additive
    scaleSpace=BCS_BoneSpace
    translationMode=BMM_Additive
    translationSpace=BCS_BoneSpace
    alphaBoolBlend {blendOption=Linear}
    alphaScaleBiasClamp {bInterpResult=true, interpSpeedDecreasing=2.5}
    boneToModify {boneName=base_collar_r}
    rotation {roll=25}
   }
   Save cached pose 'PostSimCache'(Pose=Component To Local(ComponentPose=@n18)) {cacheName=PostSimCache}
  ]}
 EventGraph {nodes [
   n1 = Event BlueprintUpdateAnimation()
   > n2 = Sequence!(execute) -> (then_0, then_1, then_2:Exec, then_3:Exec, then_4:Exec)
   > n3 = Gravity Override State Function DEPRECATED(CharacterPartAnimInstance=Self-Reference(), GameTimeInSeconds=GetGameTimeInSeconds(), GravityOverrideParams=Get Gravity Override Params())
   > Set GravityOverride(GravityOverride=@n3)
   > Set SkelMesh(SkelMesh=GetPartSkeletalMeshComponent())
   > Set leftCollar_RB_dyn(leftCollar_RB_dyn:double=float / float(A=Get headUp(), B=0.8)) -> (Output_Get:double)
   n4 = PlayerMovement(execute=@n2.then_4, AnimInstance=Self-Reference(), XDotProductCoefficient=100, YDotProductCoefficient=100, ZDotProductCoefficient=100)
   n5 = BreakVector(InVec=MakeVector(X=@n4, Y=@n4.YMovement))
   > Set Left(Left:double="To Float (Boolean)"(InBool="float < float"(A=@n5, B=-50))) -> (Output_Get:double)
   n6 = Get bIsTargeting()
   Set isTargeting(
    execute=@n2.then_1
    isTargeting:double=FInterpTo(Current=Get isTargeting() -> (isTargeting:double), Target="To Float (Boolean)"(InBool=@n6), DeltaTime=@n1.DeltaTimeX, InterpSpeed=0)
   ) -> (Output_Get:double)
   > Set Is Targeting headUp Back(Is Targeting headUp Back:double=float * float!(A=Get isTargeting(), B=Get headUp(), C:double=Get Backward() -> (Backward:double), D:bool=Get bIsCrouching()) -> (ReturnValue) {is=title K2Node_PromotableOperator}) -> (Output_Get:double)
   n7 = NOT Boolean(A=Get bIsMoving2D())
   n8 = Get bIsCrouching()
   n9 = Get isUsingDoublePistols() -> (isUsingDoublePistols:bool)
   > Set isUsingDoublePistolsCrouch(isUsingDoublePistolsCrouch:double="To Float (Boolean)"(InBool=AND Boolean!(A=@n9, B=@n8, C:bool=@n7) -> (ReturnValue))) -> (Output_Get:double)
   > Set isTargetingHeadUp(isTargetingHeadUp:double=float * float!(A=Get isTargeting(), B=Get headUp()) -> (ReturnValue)) -> (Output_Get:double)
   > Set headUp(headUp:double=GetAlphaFromReferenceTransformDelta(SkeletalMeshComponent=Get SkelMesh(), BoneName=spine_05, BaseName=head, InputRange=(X=0,Y=90), TwistAxis="0,0,1")) -> (Output_Get:double)
   > Set isUsingRocketLauncherFloat(isUsingRocketLauncherFloat:double="To Float (Boolean)"(InBool=Get isUsingRocketLauncher() -> (isUsingRocketLauncher:bool))) -> (Output_Get:double)
   n10 = AND Boolean!(A=Get bIsCrouching(), B=Get bIsMoving2D(), C=Get isUsingPistol() -> (isUsingPistol:bool), D:bool=Get bIsTargeting()) -> (ReturnValue)
   > Set isCrounchingMoving_Pistol(isCrounchingMoving_Pistol:double=float * float(A="To Float (Boolean)"(InBool=@n10), B=Get headUp())) -> (Output_Get:double)
   > n11 = Branch(Condition=Get bIsCrouching())
   > Set translateCollarBack(translateCollarBack:Vector="-5,0,0") -> (Output_Get:Vector)
   > Set rotationCollarBack(rotationCollarBack:Rotator="25,0,0") -> (Output_Get:Rotator)
   Set translateCollarBack(execute=@n11.else, translateCollarBack="0, 0, 0")
   > Set rotationCollarBack(rotationCollarBack="0,-45,0")
   n12 = Cast To FortPlayerAnimInstance(execute=@n2.then_3, Object=GetAnimInstance(self=GetOwnerSkeletalMeshComponent()))
   > Set isUsingPistol(isUsingPistol:bool=false) -> (Output_Get:bool)
   > Set isUsingDoublePistols(isUsingDoublePistols:bool=false) -> (Output_Get:bool)
   > Set isUsingRocketLauncher(isUsingRocketLauncher:bool=false) -> (Output_Get:bool)
   > n13 = Switch on EFortWeaponCoreAnimation!(execute, Selection:EFortWeaponCoreAnimation Enum=Get WeaponCoreAnim(self=@n12)) -> (
    Melee:Exec
    Pistol:Exec
    Shotgun:Exec
    PaperBlueprint:Exec
    Rifle:Exec
    MeleeOneHand:Exec
    MachinePistol:Exec
    RocketLauncher:Exec
    GrenadeLauncher:Exec
    GoingCommando:Exec
    AssaultRifle:Exec
    TacticalShotgun:Exec
    SniperRifle:Exec
    TrapPlacement:Exec
    ShoulderLauncher:Exec
    AbilityDecoTool:Exec
    Crossbow:Exec
    C4:Exec
    RemoteControl:Exec
    DualWield:Exec
    AR_BullPup:Exec
    AR_ForwardGrip:Exec
    MedPackPaddles:Exec
    SMG_P90:Exec
    AR_DrumGun:Exec
    Consumable_Small:Exec
    Consumable_Large:Exec
    Balloon:Exec
    MountedTurret:Exec
    CreativeTool:Exec
    ExplosiveBow:Exec
    AshtonIndigo:Exec
    AshtonChicago:Exec
    MeleeDualWield:Exec
    Unarmed:Exec
   )
   Set isUsingDoublePistols(execute=@n13.DualWield, isUsingDoublePistols=true)
   Set isUsingPistol(execute=@n13.Pistol, isUsingPistol=true)
   Set isUsingRocketLauncher(execute=@n13.RocketLauncher, isUsingRocketLauncher=true)
   n14 = Set isWearingCape(execute=@n2.then_2, isWearingCape=CheckForBackbling().bIsBackblingLong)
   n15 = Get bIsCrouching()
   > Set isWearingCape2(isWearingCape2:bool=AND Boolean!(A=@n14, B=NOT Boolean(A=@n15)) -> (ReturnValue)) -> (Output_Get:bool)
   n16 = Get isWearingCape() -> (isWearingCape:bool)
   > n17 = Branch(Condition=AND Boolean(A=NOT Boolean(A=@n16), B=Get bIsCrouching()))
   > Set isPouchingRotate(isPouchingRotate:double=1) -> (Output_Get:double)
   n18 = Branch(execute=@n17.else, Condition=AND Boolean(A=Get isWearingCape(), B=Get bIsCrouching()))
   > Set isPouchingRotate(isPouchingRotate=0.25)
   Set isPouchingRotate(
    execute=@n18.else
    isPouchingRotate=GetAlphaFromReferenceTransformDelta(SkeletalMeshComponent=Get SkelMesh(), BoneName=pelvis, BaseName=thigh_l, InputRange=(X=0,Y=90), OutputRange=(X=0,Y=0.5), TwistAxis="0,0,1")
   )
   Event BlueprintInitializeAnimation()
   > Set SkelMesh(SkelMesh=GetOwnerSkeletalMeshComponent())
   Set Right(Right:double="To Float (Boolean)"(InBool="float > float"(A=@n5, B=50))) -> (Output_Get:double)
   > Set Forward(Forward:double="To Float (Boolean)"(InBool="float > float"(A=@n5.Y, B=40))) -> (Output_Get:double)
   > Set Backward(Backward:double="To Float (Boolean)"(InBool="float < float"(A=@n5.Y, B=-40))) -> (Output_Get:double)
  ]}
]
variables [
 Backward {type=number}
 Forward {type=number}
 Gravity Override Params {type=GravityOverrideParamsStruct}
 GravityOverride {type=Vector}
 headUp {type=number}
 Is Targeting headUp Back {type=number}
 isCrounchingMoving_Pistol {type=number}
 isHeadUpCrouching {type=number}
 isPouchingRotate {type=number}
 isTargeting {type=number}
 isTargetingHeadUp {type=number}
 isUsingDoublePistols {type=boolean}
 isUsingDoublePistolsCrouch {type=number}
 isUsingPistol {type=boolean}
 isUsingRocketLauncher {type=boolean}
 isUsingRocketLauncherFloat {type=number}
 isWearingCape {type=boolean}
 isWearingCape2 {type=boolean}
 Left {type=number}
 leftCollar_RB_dyn {type=number}
 Right {type=number}
 rotationCollarBack {type=Rotator}
 SkelMesh {type=SkeletalMeshComponent}
 translateCollarBack {type=Vector}
]
